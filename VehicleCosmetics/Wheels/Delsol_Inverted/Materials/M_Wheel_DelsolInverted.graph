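# Unlit, emissive-only material for the Delsol Inverted vehicle wheel, with animated, tintable glowing rim and tire details.
# Rim masks shape panning gradient stripes and a circle, while the tread combines scrolling noise-broken lines, normal-mapped Fresnel terms and a tread mask.
# Vertex color red blends rim and tread regions, tinted by CustomColor and CircleColor with separate brightness controls.
# The emissive result runs through vehicle damage and de-rez transition functions.
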Material /VehicleCosmetics/Wheels/Delsol_Inverted/Materials/M_Wheel_DelsolInverted
graph {
 properties {
  bAutomaticallySetUsageInEditor=false
  bUseMaterialAttributes=true
  bUsedWithMeshDeformer=true
  bUsedWithSkeletalMesh=true
  previewMesh=/VehicleCosmetics/Wheels/Delsol_Inverted/Meshes/Wheel_DelSol_Inverted_SM
  shadingModel=MSM_Unlit
  usedShadingModels=Unlit
 }
 nodes [
  n1 = TextureCoordinate() {vTiling=4}
  n2 = TextureSampleParameter2D() {parameterName=Rim Masks, texture=/VehicleCosmetics/Wheels/Delsol_Inverted/Textures/T_DelSolInverted_Rim_M}
  n3 = Multiply(
   A=TextureSample(UVs=Panner(Coordinate=@n1) {speedX=-0.5, speedY=0.5}) {samplerSource=SSM_Wrap_WorldGroupSettings, texture=/VehicleCosmetics/Wheels/_SharedTextures/FX/GradientStripe2_T}.RGB
   B=@n2.B
  )
  n4 = Clamp(None=Add(A=Multiply(A=@n3, B=@n2.G), B=@n2.R))
  n5 = Multiply(A=@n2.A) {constB=5}
  Add(A=@n4, B=@n5)
  n6 = NamedRerouteUsage()
  MaterialFunctionCall(In=MaterialFunctionCall(Material=SetMaterialAttributes(Emissive Color=@n6) {attributeSetTypes [#1]}) {materialFunction=&MF_BaseVehicle_Damage}.Result) {materialFunction=&MF_VehicleDeRez_Transition}
  NamedRerouteDeclaration(None=@n4) {name=Tread Details, nodeColor=<rgba 0.0547 1 0 1>}
  n7 = TextureCoordinate() {uTiling=12}
  n8 = ConstantBiasScale(None=TextureSample(UVs=Panner(Coordinate=@n7) {speedX=-1}) {samplerSource=SSM_Wrap_WorldGroupSettings, texture=/VehicleCosmetics/Wheels/_SharedTextures/FX/LinearGradient_Balanced_Vertical}.G) {bias=-0.2, scale=10}
  n9 = TextureCoordinate() {uTiling=0.2}
  n10 = Multiply(
   A=@n8
   B=TextureSample(UVs=Panner(Coordinate=@n9) {speedX=-0.5}) {samplerSource=SSM_Wrap_WorldGroupSettings, texture=/VehicleCosmetics/Wheels/_SharedTextures/FX/Universal_Noise_34_RGB}.B
  )
  n11 = TextureSample() {samplerSource=SSM_Wrap_WorldGroupSettings, texture=/VehicleCosmetics/Wheels/Delsol_Inverted/Textures/T_DelsolInverted_Tire_BC}
  n12 = Add(A=@n11.RGB, B=@n11.RGB)
  NamedRerouteDeclaration(None=Multiply(A=Multiply(A=@n10, B=@n12)) {constB=2}) {name=Tread Lines Final, nodeColor=<rgba 1 0 0.375 1>}
  NamedRerouteDeclaration(None=@n8) {name=Tread Lines Bias, nodeColor=<rgba 1 0 0.375 1>}
  n13 = TextureSampleParameter2D() {
   parameterName=Tread Normal
   samplerType=SAMPLERTYPE_Normal
   texture=/VehicleCosmetics/Wheels/Delsol_Inverted/Textures/T_DelsolInverted_Tire_N
  }
  n14 = Constant() {r=5}
  NamedRerouteDeclaration(None=Multiply(A=MaterialFunctionCall!(Normal=@n13.RGB, Exponent=@n14) {materialFunction=/VehicleCosmetics/Wheels/Cola_Holographic/Materials/MF_Fresnel_Classic}.Result) {constB=0.5}) {name=Fresnel 1, nodeColor=<rgba 1 0.9844 0 1>}
  n15 = Constant() {r=2.5}
  NamedRerouteDeclaration(None=Multiply(A=MaterialFunctionCall(Normal=@n13.RGB, Exponent=@n15) {materialFunction=&MF_Fresnel_Classic}.Result) {constB=10}) {name=Fresnel 2, nodeColor=<rgba 1 0.9844 0 1>}
  n16 = LinearInterpolate(A=Constant3Vector() {constant=<rgba 0.15 0.15 0.15 0>}, B=Constant3Vector(), Alpha=NamedRerouteUsage())
  n17 = Clamp(None=NamedRerouteUsage()) {maxDefault=0.5, minDefault=0.2}
  n18 = TextureSampleParameter2D() {parameterName=Tread Mask, texture=/VehicleCosmetics/Wheels/Delsol_Inverted/Textures/T_DelsolInverted_Tire_M}
  n19 = Add(A=NamedRerouteUsage(), B=Add(A=Multiply(A=@n16, B=@n17), B=Multiply(A=@n18.R) {constB=6}))
  NamedRerouteDeclaration(None=Add(A=NamedRerouteUsage(), B=Clamp(None=@n19) {maxDefault=6})) {name=Tread Assembly, nodeColor=<rgba 1 0.9844 0 1>}
  n20 = VectorParameter() {parameterName=CustomColor, defaultValue=<rgba 0 0.3467 1 1>}
  n21 = LinearInterpolate(A=Multiply(A=@n20.RGB, B=NamedRerouteUsage()), B=Multiply(A=@n20.RGB, B=NamedRerouteUsage()), Alpha=VertexColor().R)
  n22 = ScalarParameter() {defaultValue=1, parameterName=Global brightness}
  n23 = VectorParameter() {parameterName=CircleColor, defaultValue=<rgba 0 0.3135 1 1>}
  n24 = LinearInterpolate(
   A=Multiply(A=NamedRerouteUsage(), B=@n23.RGB)
   B=Multiply(
    A=TextureSampleParameter2D() {parameterName=Tread Mask, texture=&T_DelsolInverted_Tire_M}.R
    B=@n23.RGB
   )
   Alpha=VertexColor().R
  )
  n25 = ScalarParameter() {defaultValue=2, parameterName=CircleBrightness}
  NamedRerouteDeclaration(None=Add(A=Multiply(A=@n21, B=@n22), B=Multiply(A=@n24, B=@n25))) {name=Emissive, nodeColor=<rgba 0.1016 0 1 1>}
  NamedRerouteDeclaration(None=@n5) {name=RimCircle, nodeColor=<rgba 0.1953 1 0 1>}
 ]
}
